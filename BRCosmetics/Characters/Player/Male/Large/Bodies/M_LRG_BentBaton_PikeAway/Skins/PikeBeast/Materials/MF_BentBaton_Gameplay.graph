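# Material function for the PikeBeast skin of BentBaton that layers an animated glowing, subsurface effect onto incoming material attributes and outputs its emissive area mask.
# Two parallax swirl-noise layers scroll over time on pre-skinned positions, in per-area directions.
# Area-masked fresnels build a mask that indexes a color-scheme texture.
# Skylight fake highlights, AO and chest-distance falloff shape the emissive, alongside overrides to base color, specular, roughness, metallic and opacity.
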MaterialFunction /BRCosmetics/Characters/Player/Male/Large/Bodies/M_LRG_BentBaton_PikeAway/Skins/PikeBeast/Materials/MF_BentBaton_Gameplay
graph {nodes [
  CameraPositionWS()
  ComponentMask(None=VertexNormalWS()) {g=true}
  n1 = NamedRerouteUsage()
  n2 = NamedRerouteUsage()
  n3 = NamedRerouteUsage()
  n4 = MaterialFunctionCall(Float3=@n3) {materialFunction=&BreakOutFloat3Components}
  n5 = Max(A=@n1, B=Max(A=@n2, B=NamedRerouteDeclaration(None=@n4.B) {name=Area Mask - Three, nodeColor=<rgba 1 0 0.7031 1>}))
  n6 = StaticComponentMaskParameter(None=@n3) {defaultG=true, parameterName=Emissive Skin Mask}
  FunctionOutput(None=NamedRerouteDeclaration(None=StaticSwitchParameter(True=@n5, False=@n6) {defaultValue=true, parameterName=Area/Singular Mask}) {name=Emissive Areas, nodeColor=<rgba 1 0.6328 0 1>}) {outputName=Emissive Area, sortPriority=1}
  n7 = GetMaterialAttributes(None=FunctionInput() {inputName=Result, inputType=FunctionInput_MaterialAttributes}) {attributeGetTypes [
    #1
    #2
    #3
    #4
    #5
   ]}
  n8 = SetMaterialAttributes!(
   MaterialAttributes=@n7.MaterialAttributes
   EmissiveColor=NamedRerouteUsage()
   BaseColor=NamedRerouteUsage()
   Specular=NamedRerouteUsage()
   Roughness=NamedRerouteUsage()
   Opacity=NamedRerouteUsage()
   SubsurfaceColor=NamedRerouteUsage()
   Metallic=NamedRerouteUsage()
  ) {attributeSetTypes [
    #1
    #2
    #3
    #4
    #6
    #7
    #5
   ]}
  FunctionOutput(None=@n8) {outputName=SetMaterialAttributes}
  n9 = CameraVectorWS()
  n10 = Transform(None=@n9) {transformSourceType=TRANSFORMSOURCE_World, transformType=TRANSFORM_Local}
  n11 = ComponentMask(None=RotateAboutAxis(
    NormalizedRotationAxis=Constant3Vector()
    RotationAngle=ScalarParameter() {defaultValue=1, parameterName=Angle}
    PivotPoint=ObjectPositionWS()
    Position=@n10
   )) {g=true}
  n12 = OneMinus(None=@n11)
  Max(A=@n12, B=@n11)
  n13 = ComponentMask(None=Add(A=Transform(None=@n9) {transformSourceType=TRANSFORMSOURCE_World, transformType=TRANSFORM_Local}) {constB=-0.1}) {r=true}
  n14 = Max(A=@n13, B=Subtract(A=OneMinus(None=@n13)) {constB=1.1})
  n15 = LinearInterpolate(
   A=ScalarParameter() {defaultValue=0.1448, parameterName=Third Fresnel - Step Min}
   B=ScalarParameter() {defaultValue=-0.4101, parameterName=Third Fresnel - Step Max}
   Alpha=LinearInterpolate(B=@n14, Alpha=ScalarParameter() {defaultValue=1, parameterName=Third Fresnel - Camera Vector - Influence}) {constA=1}
  )
  n16 = ScalarParameter() {defaultValue=2, parameterName=Third Fresnel - Smoothness}
  n17 = ScalarParameter() {defaultValue=1, parameterName=ExponentIn}
  n18 = Saturate(None=SmoothStep(Min=@n15, Max=Add(A=@n15, B=@n16), Value=Fresnel(ExponentIn=@n17) {baseReflectFraction=0}))
  n19 = NamedRerouteUsage()
  n20 = NamedRerouteUsage()
  n21 = LinearInterpolate(B=Saturate(None=NamedRerouteUsage()), Alpha=ScalarParameter() {defaultValue=1, parameterName=Second Fresnel - Camera Vector - Influence}) {constA=1}
  n22 = LinearInterpolate(
   A=ScalarParameter() {defaultValue=-0.4176, parameterName=Second Fresnel - Step Min}
   B=ScalarParameter() {defaultValue=-0.502, parameterName=Second Fresnel - Step Max}
   Alpha=@n21
  )
  n23 = ScalarParameter() {defaultValue=3, parameterName=Second Fresnel - Tightness Min}
  n24 = ScalarParameter() {defaultValue=3, parameterName=Second Fresnel - Tightness Max}
  n25 = OneMinus(None=MaterialFunctionCall(Exponent=ScalarParameter() {defaultValue=1, parameterName=Fresnel Power}) {materialFunction=&AxisAlignedFresnel}.Fresnel)
  n26 = SmoothStep(Min=@n22, Max=Add(A=@n22, B=LinearInterpolate(A=@n23, B=@n24, Alpha=@n21)), Value=@n25)
  n27 = Saturate(None=LinearInterpolate(A=@n26, B=OneMinus(None=@n26), Alpha=ScalarParameter() {parameterName=Second Fresnel - Flip}))
  n28 = Max(
   A=Multiply(A=Multiply(A=@n18, B=@n19), B=ScalarParameter() {defaultValue=1, parameterName=Third Fresnel - Intensity})
   B=Multiply(A=Multiply(A=@n20, B=@n27), B=ScalarParameter() {defaultValue=1, parameterName=Second Fresnel - Intensity})
  )
  n29 = ScalarParameter() {defaultValue=-0.3886, parameterName=Fresnel Step Min}
  n30 = SmoothStep(Min=@n29, Max=Add(A=@n29, B=ScalarParameter() {defaultValue=4, parameterName=Fresnel Tightness}), Value=@n25)
  n31 = Saturate(None=LinearInterpolate(A=@n30, B=OneMinus(None=@n30), Alpha=ScalarParameter() {parameterName=First Fresnel - Flip}))
  n32 = Multiply(A=ScalarParameter() {defaultValue=1, parameterName=First Fresnel - Intensity}, B=@n31)
  n33 = Multiply(A=NamedRerouteUsage(), B=@n32)
  n34 = Max(A=StaticSwitchParameter(True=Max(A=@n28, B=@n33), False=@n32) {defaultValue=true, parameterName=Area/Singular Mask}, B=NamedRerouteUsage())
  n35 = NamedRerouteUsage()
  n36 = ScalarParameter() {defaultValue=3, parameterName=Swirl Noise - Highlight Power}
  Max(
   A=@n34
   B=Multiply(A=Power(Base=@n35, Exp=@n36) {constExponent=1}, B=ScalarParameter() {defaultValue=1, parameterName=Swirl Noise - Highlight Intensity})
  )
  n37 = ComponentMask(None=MaterialFunctionCall!(VectorToRotate=@n10, LookAtVector=LocalPosition().XYZ, RestingVector) {materialFunction=/Engine/Functions/Engine_MaterialFunctions01/Vectors/RotateVector}.Result) {r=true}
  n38 = Max(
   A=Add(A=@n37, B=ScalarParameter() {defaultValue=-0.4, parameterName=Camera Vec - Distance}) {constB=-0.5}
   B=Add(A=OneMinus(None=@n37), B=ScalarParameter() {defaultValue=-1.4, parameterName=B_1})
  )
  n39 = ScalarParameter() {defaultValue=5.5, parameterName=B}
  Max(A=@n38, B=Multiply(A=Add(A=@n12, B=@n39) {constB=0.8}))
  n40 = FunctionInput() {inputName=M}
  n41 = Power(
   Base=ChannelMaskParameter(None=@n40) {maskChannel=Green, parameterName=AO, defaultValue=<rgba 0 1 0 0>}.R
   Exp=ScalarParameter() {defaultValue=1, parameterName=Exp}
  )
  n42 = ScalarParameter() {defaultValue=0.3, parameterName=Intensity}
  Min(A=@n41, B=@n42)
  n43 = OneMinus(None=Fresnel(ExponentIn=ScalarParameter() {defaultValue=0.75, parameterName=Opacity - Exp}))
  n44 = ScalarParameter() {defaultValue=2, parameterName=Opacity - Power}
  n45 = ScalarParameter() {defaultValue=0.75, parameterName=Opacity - Top Down Exp}
  n46 = Multiply(
   A=VectorParameter() {parameterName=SSS Color, defaultValue=<rgba 2 0.1042 0.1476 1>}.RGB
   B=Multiply(A=Power(Base=@n43, Exp=@n44), B=Fresnel(ExponentIn=@n45))
  )
  n47 = NamedRerouteUsage()
  n48 = Power(
   Base=Multiply(A=OneMinus(None=@n47), B=ScalarParameter() {defaultValue=1.2581, parameterName=Subsurface - Noise Strength})
   Exp=ScalarParameter() {defaultValue=0.6639, parameterName=Subsurface - Noise Power}
  )
  Multiply(A=@n46, B=@n48)
  NamedRerouteDeclaration(None=@n14) {name=Camera Vector Mask, nodeColor=<rgba 0.7109 1 0 1>}
  NamedRerouteDeclaration(None=@n38) {name=Neck Camera Vector, nodeColor=<rgba 0.9688 0 1 1>}
  NamedRerouteDeclaration(None=LinearInterpolate(A=NamedRerouteUsage(), B=ScalarParameter() {parameterName=Emissive - Metallic}, Alpha=NamedRerouteUsage())) {name=Metallic, nodeColor=<rgba 0 0.6484 1 1>}
  n49 = Fresnel(
   ExponentIn=ScalarParameter() {defaultValue=0.7831, parameterName=ExponentIn_1}
   BaseReflectFractionIn=ScalarParameter() {defaultValue=0.7086, parameterName=BaseOpacity}
  )
  n50 = Multiply(
   A=LinearInterpolate(A=@n49, B=OneMinus(None=@n49), Alpha=ScalarParameter() {defaultValue=1, parameterName=Flip SSS Fresnel})
   B=NamedRerouteUsage()
  )
  n51 = Multiply(A=@n48, B=ChannelMaskParameter(None=@n40) {maskChannel=Green, parameterName=Curvatrue, defaultValue=<rgba 0 1 0 0>}.R)
  NamedRerouteDeclaration(None=Saturate(None=Multiply(A=@n50, B=@n51))) {name=Opacity, nodeColor=<rgba 0.9688 1 0 1>}
  NamedRerouteDeclaration(None=LinearInterpolate(A=NamedRerouteUsage(), B=NamedRerouteUsage(), Alpha=NamedRerouteUsage())) {name=Emissive Color, nodeColor=<rgba 1 0.8672 0 1>}
  NamedRerouteDeclaration(None=LinearInterpolate(A=NamedRerouteUsage(), Alpha=NamedRerouteUsage()) {constB=0}) {name=Base Color, nodeColor=<rgba 0.5469 0 1 1>}
  n52 = TextureSampleParameter2D() {parameterName=FX2, texture=/BRCosmetics/Weapons/FORT_Melee/Pickaxe_BentBaton/Skins/Textures/T_BentBaton_Pickaxe_Flipaway_FX2}
  NamedRerouteDeclaration(None=@n52.B) {name=Glow Edges, nodeColor=<rgba 0 0.5078 1 1>}
  NamedRerouteDeclaration(None=@n7.EmissiveColor) {name=Emissive Color - In, nodeColor=<rgba 0.4531 0 1 1>}
  NamedRerouteDeclaration(None=@n7.BaseColor) {name=Base Color - In, nodeColor=<rgba 0 1 0.1328 1>}
  NamedRerouteDeclaration(None=@n7.Specular) {name=Specular - In, nodeColor=<rgba 1 0 0.1875 1>}
  NamedRerouteDeclaration(None=@n7.Roughness) {name=Roughness - In, nodeColor=<rgba 0 0.5078 1 1>}
  NamedRerouteDeclaration(None=@n7.Metallic) {name=Metallic In, nodeColor=<rgba 0.8281 1 0 1>}
  n53 = Subtract(A=@n34, B=Multiply(A=@n31, B=ScalarParameter() {defaultValue=1, parameterName=Edge Fresnel Removal}))
  NamedRerouteDeclaration(None=@n53) {name=Emissive Mask, nodeColor=<rgba 0.5234 1 0 1>}
  n54 = TextureSampleParameter2D(UVs=@n53) {
   parameterName=Color Scheme
   samplerType=SAMPLERTYPE_LinearColor
   texture=/BRCosmetics/Weapons/FORT_Melee/Pickaxe_BentBaton/Skins/Materials/CA_Pickaxe_Bentbaton_Flipaway
  }
  n55 = LinearInterpolate(
   A=@n54.RGB
   B=LinearInterpolate(A=NamedRerouteUsage(), B=@n54.RGB, Alpha=@n53)
   Alpha=ScalarParameter() {defaultValue=1, parameterName=Blend Original Emissive}
  )
  n56 = Multiply(
   A=Fresnel(ExponentIn=ScalarParameter() {defaultValue=2.5, parameterName=Edge Highlight - Exp}) {baseReflectFraction=0}
   B=ScalarParameter() {defaultValue=0.1, parameterName=Edge Highlight - Intensity}
  )
  n57 = ScalarParameter() {defaultValue=0.7449, parameterName=Fake Highlight - Step}
  n58 = ScalarParameter() {defaultValue=1, parameterName=Fake Highlight - Smoothness}
  n59 = ReflectionVectorWS()
  n60 = DotProduct(A=@n59, B=VertexNormalWS())
  n61 = ScalarParameter() {defaultValue=1, parameterName=Fake Highlight - Exp}
  n62 = Multiply(A=SmoothStep(Min=@n57, Max=Add(A=@n57, B=@n58), Value=Power(Base=@n60, Exp=@n61)), B=SkyLightEnvMapSample(Direction=@n59))
  n63 = ScalarParameter() {defaultValue=1, parameterName=Fake Highlight Strength}
  n64 = MaterialFunctionCall(Base=Add(A=Add(A=@n55, B=@n56), B=Multiply(A=@n62, B=@n63)), Blend=Max(A=@n41, B=@n42)) {materialFunction=&Blend_Overlay}
  n65 = Subtract(A=VectorParameter() {group=Override, parameterName=chestPOS, defaultValue=<rgba 0 0 0 1>}.RGB, B=NamedRerouteUsage())
  n66 = OneMinus(None=Divide(A=Length(None=@n65) {is=class MaterialExpressionLength}, B=ScalarParameter() {defaultValue=100, group=Override, parameterName=ChestRadius}) {constB=10})
  n67 = LinearInterpolate(
   A=ScalarParameter() {defaultValue=0.3, group=Override, parameterName=MinColor}
   Alpha=SmoothStep(
    Min=ScalarParameter() {defaultValue=0.5, group=Override, parameterName=MinSmooth}
    Max=ScalarParameter() {defaultValue=1, group=Override, parameterName=MaxSmooth}
    Value=Power(Base=@n66)
   )
  )
  NamedRerouteDeclaration(None=Multiply(A=Multiply(A=@n64.Result, B=@n67), B=ScalarParameter() {defaultValue=1, parameterName=Emissive Intensity})) {name=Emissive Effect, nodeColor=<rgba 0.8516 0 1 1>}
  n68 = Multiply(A=NamedRerouteUsage(), B=VectorParameter() {parameterName=Head Direction, defaultValue=<rgba 0 2 -5 1>}.RGB) {constB=2}
  n69 = VertexInterpolator(VS=LocalPosition() {localOrigin=InstancePreSkinning}.XYZ)
  n70 = Multiply(A=NamedRerouteUsage(), B=VectorParameter() {parameterName=Handle Direction, defaultValue=<rgba 0 -1 -5 1>}.RGB) {constB=2}
  n71 = Add(A=Multiply(A=Add(A=@n68, B=@n69.PS), B=NamedRerouteUsage()), B=Multiply(A=Add(A=@n69.PS, B=@n70), B=NamedRerouteUsage()))
  n72 = Add(
   A=@n69.PS
   B=Multiply(A=NamedRerouteUsage(), B=VectorParameter() {parameterName=Flow Map - Direction, defaultValue=<rgba 0 -1 -5 1>}.RGB) {constB=2}
  )
  n73 = ScalarParameter() {parameterName=Global Direction}
  NamedRerouteDeclaration(None=Add(A=LinearInterpolate(A=@n71, B=@n72, Alpha=@n73), B=NamedRerouteUsage())) {name=Local Position Offset, nodeColor=<rgba 0.125 0 1 1>}
  NamedRerouteDeclaration(None=Multiply(A=ScalarParameter() {defaultValue=0.4, parameterName=Head Rotation}, B=NamedRerouteUsage())) {name=Rotated Head UVs, nodeColor=<rgba 0.1953 1 0 1>}
  NamedRerouteDeclaration(None=VertexInterpolator(VS=PreSkinnedNormal()).PS) {name=Local Normal, nodeColor=<rgba 1 0 0.5156 1>}
  NamedRerouteDeclaration(None=FunctionInput() {sortPriority=4}) {name=PreskinnedPos, nodeColor=<rgba 0 1 0.7422 1>}
  NamedRerouteDeclaration(None=TextureObject() {texture=/BRCosmetics/Weapons/FORT_Melee/Pickaxe_BentBaton/Materials/Swirly_Wisps_Vertical_Dissolve}) {name=Texture Obj, nodeColor=<rgba 1 0.8438 0 1>}
  n74 = ScalarParameter() {parameterName=Y Rotation}
  n75 = ScalarParameter() {parameterName=Z Rotation}
  n76 = ScalarParameter() {defaultValue=0.18, parameterName=X Rotation}
  n77 = ScalarParameter() {defaultValue=-0.35, parameterName=Swirl - Main Layer - Depth}
  n78 = Multiply(
   A=MaterialFunctionCall!(
    RotYAxis=@n74
    TextureObject=NamedRerouteUsage()
    TextureSize=ScalarParameter() {defaultValue=15, parameterName=Swirl - Tiling}
    WorldPosition=NamedRerouteUsage()
    Export Float 4
    Heightmap=NamedRerouteUsage()
    RotZAxis=@n75
    RotXAxis=@n76
    World Space Normal=NamedRerouteUsage()
    Parallax=@n77
    Height Amount=@n77
    Min Steps
    Max Steps
    Ref Plane
    Bump Amount=@n77
    ProjectionTransitionContrast=ScalarParameter() {parameterName=Swirl - Projection Transition Contrast}
   ) {materialFunction=/BRCosmetics/Weapons/FORT_Melee/Pickaxe_BentBaton/Materials/MF_WorldAlignedTexture_RotBumpOffset}.XYZ Texture
   B=ScalarParameter() {defaultValue=1, parameterName=Swirl Noise - Main Opacity}
  )
  n79 = ScalarParameter() {defaultValue=-0.1, parameterName=Swirl - Second Layer - Depth}
  n80 = MaterialFunctionCall(
   RotYAxis=@n74
   TextureObject=NamedRerouteUsage()
   TextureSize=ScalarParameter() {defaultValue=15, parameterName=Swirl - Tiling}
   WorldPosition=Add(A=NamedRerouteUsage(), B=ScalarParameter() {defaultValue=4.875, parameterName=Swirl - Second Layer - Offset}) {constB=6}
   RotZAxis=@n75
   RotXAxis=@n76
   World Space Normal=NamedRerouteUsage()
   Parallax=@n79
   Height Amount=@n79
   Bump Amount=@n79
   ProjectionTransitionContrast=ScalarParameter() {parameterName=Swirl - Projection Transition Contrast}
  ) {materialFunction=&MF_WorldAlignedTexture_RotBumpOffset}
  n81 = ScalarParameter() {defaultValue=1, parameterName=Swirl - Second Layer - Opacity}
  n82 = NamedRerouteDeclaration(None=Max(A=@n78, B=Multiply(A=@n80.XYZ Texture, B=@n81))) {name=Swirl Noise, nodeColor=<rgba 0.2656 0 1 1>}
  n83 = Multiply(
   A=Power(Base=@n82, Exp=ScalarParameter() {defaultValue=3, parameterName=Swirl Noise - Highlight Power}) {constExponent=1}
   B=ScalarParameter() {defaultValue=1, parameterName=Swirl Noise - Highlight Intensity}
  )
  n84 = Multiply(
   A=Power(Base=@n82, Exp=ScalarParameter() {defaultValue=0.5, parameterName=Swirl Noise - Power})
   B=ScalarParameter() {defaultValue=0.4, parameterName=Swirl - Noise Strength}
  )
  n85 = ScalarParameter() {parameterName=Swirl Noise - Edge Removal - Step}
  n86 = Add(A=@n85, B=ScalarParameter() {defaultValue=2, parameterName=Swirl Noise - Edge Removal - Smoothness})
  n87 = Fresnel(ExponentIn=ScalarParameter() {defaultValue=2, parameterName=Swirl Noise - Edge Removal - Exp}) {baseReflectFraction=0}
  NamedRerouteDeclaration(None=Subtract(A=Max(A=@n83, B=@n84), B=SmoothStep(Min=@n85, Max=@n86, Value=@n87))) {name=Swirl Noise - Cut, nodeColor=<rgba 1 0.5859 0 1>}
  NamedRerouteDeclaration(None=@n52.R) {name=Area Mask - One, nodeColor=<rgba 0.3828 1 0 1>}
  NamedRerouteDeclaration(None=Add(A=@n1, B=@n2)) {name=Rotated Parts, nodeColor=<rgba 1 0.2812 0 1>}
  NamedRerouteDeclaration(None=@n52.G) {name=Area Mask - Two, nodeColor=<rgba 0.9688 0 1 1>}
  NamedRerouteDeclaration(None=FunctionInput() {inputName=Tech Art Mask, sortPriority=10}) {name=Tech Art Mask, nodeColor=<rgba 1 0 0.9141 1>}
  NamedRerouteDeclaration(None=@n78) {name=SSS Mask, nodeColor=<rgba 0.3594 1 0 1>}
  NamedRerouteDeclaration(None=Multiply(A=Time(), B=ScalarParameter() {defaultValue=0.4, parameterName=Global Speed})) {name=Global Time, nodeColor=<rgba 0.2422 1 0 1>}
  NamedRerouteDeclaration(None=LinearInterpolate(A=NamedRerouteUsage(), B=ScalarParameter() {defaultValue=2, parameterName=Emissive - Spec}, Alpha=NamedRerouteUsage())) {name=Specular, nodeColor=<rgba 0.3125 0 1 1>}
  NamedRerouteDeclaration(None=LinearInterpolate(A=NamedRerouteUsage(), B=ScalarParameter() {defaultValue=0.3, parameterName=Emissive - Roughness}, Alpha=NamedRerouteUsage())) {name=Roughness, nodeColor=<rgba 0.0078 1 0 1>}
  NamedRerouteDeclaration(None=LinearInterpolate(B=@n46, Alpha=@n50)) {name=SSS Color, nodeColor=<rgba 1 0 0.3281 1>}
  OneMinus(None=NamedRerouteUsage())
  ScalarParameter() {defaultValue=-0.0567, parameterName=Fresnel Step Max}
 ]}
