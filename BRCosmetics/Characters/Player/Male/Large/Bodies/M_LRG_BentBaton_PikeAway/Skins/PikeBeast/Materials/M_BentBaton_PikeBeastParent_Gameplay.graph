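# Gameplay parent material for the PikeBeast skin of the BentBaton PikeAway large male player character.
# It runs the shared 2023 character PreFX function, a BentBaton gameplay function, then the shared PostFX function.
# Between them, emissive is desaturated and scaled by PostDesaturation and PostEM as IsActive rises, applied only within the gameplay function's Emissive Area mask.
# It uses material attributes and supports skeletal meshes, morph targets and mesh deformers.
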Material /BRCosmetics/Characters/Player/Male/Large/Bodies/M_LRG_BentBaton_PikeAway/Skins/PikeBeast/Materials/M_BentBaton_PikeBeastParent_Gameplay
graph {
 properties {
  bHasPixelAnimation=true
  bUseMaterialAttributes=true
  bUsedWithMeshDeformer=true
  bUsedWithMorphTargets=true
  bUsedWithSkeletalMesh=true
  pixelDepthOffsetMode=PDOM_Legacy
  usedShadingModels=Default Lit
 }
 nodes [
  n1 = MaterialFunctionCall() {materialFunction=&MF_CharacterParent_2023_PreFX}
  n2 = MaterialFunctionCall!(M=@n1.M, Result=@n1.M, In=@n1.M, Tech Art Mask=@n1.M) {materialFunction=&/MF_BentBaton_Gameplay}
  n3 = GetMaterialAttributes(None=@n2.SetMaterialAttributes) {attributeGetTypes [#1]}
  n4 = ScalarParameter() {parameterName=PostDesaturation}
  n5 = ScalarParameter() {parameterName=IsActive}
  n6 = Multiply(
   A=Desaturation(None=@n3.Emissive Color, Fraction=LinearInterpolate(B=@n4, Alpha=@n5) {constB=0})
   B=LinearInterpolate(A=Constant() {r=1}, B=ScalarParameter() {defaultValue=1, parameterName=PostEM}, Alpha=@n5) {constA=1}
  )
  MaterialFunctionCall(
   Material=SetMaterialAttributes(MaterialAttributes=@n3.MaterialAttributes, Emissive Color=LinearInterpolate(A=@n3.Emissive Color, B=@n6, Alpha=@n2.Emissive Area)) {attributeSetTypes [#1]}
   M=@n1.M
   Pre Skinned Local Position=@n1.M
   Distance Blend=@n1.M
   Fresnel=@n1.M
  ) {materialFunction=&MF_CharacterParent_2023_PostFX}
 ]
}
